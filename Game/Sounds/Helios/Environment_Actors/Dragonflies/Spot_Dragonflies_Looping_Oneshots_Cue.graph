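# Looping ambient SoundCue for dragonfly fly-by one-shots, routed through the Athena ambient sound class with attenuation, an exterior ambient bus send and a reverb send.
# A Mixer layers three endless loops, each with its own random delay range, pitch and volume modulation, and random pick from four fly-by waves.
# The Quality Level root connects that mix to its first two slots only, leaving the other two silent.
# A concurrency override allows one instance at a time.
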SoundCue /Game/Sounds/Helios/Environment_Actors/Dragonflies/Spot_Dragonflies_Looping_Oneshots_Cue
root=QualityLevel_0
nodes [
 {id=Delay_0, children [Modulator_3], properties {delayMax=3, delayMin=1}}
 {id=Delay_2, children [Modulator_0], properties {delayMax=3.5, delayMin=1.5}}
 {id=Delay_3, children [Modulator_1], properties {delayMax=1.5, delayMin=0.5}}
 {id=Looping_0, children [Delay_2]}
 {id=Looping_2, children [Delay_3]}
 {id=Looping_3, children [Delay_0]}
 {id=Mixer_0, children [Looping_0,Looping_3,Looping_2], properties {inputVolume [1,1,1]}}
 {id=Modulator_0, children [Random_5], properties {pitchMax=1.5, pitchMin=1, volumeMax=1.1, volumeMin=0.9}}
 {id=Modulator_1, children [Random_7], properties {pitchMax=1.5, pitchMin=1, volumeMax=1.1, volumeMin=0.9}}
 {id=Modulator_3, children [Random_4], properties {pitchMax=1.5, pitchMin=1, volumeMax=1.1, volumeMin=0.9}}
 {id=QualityLevel_0, children [Mixer_0,Mixer_0,null,null]}
 {id=Random_4, children [WavePlayer_43, WavePlayer_44, WavePlayer_45, WavePlayer_46], properties {weights [1,1,1,1]}}
 {id=Random_5, children [WavePlayer_49, WavePlayer_50, WavePlayer_51, WavePlayer_52], properties {weights [1,1,1,1]}}
 {id=Random_7, children [WavePlayer_56, WavePlayer_55, WavePlayer_58, WavePlayer_57], properties {weights [1,1,1,1]}}
 {id=WavePlayer_43, properties {soundWaveAssetPtr=&/Spot_DragonFly_By_A_07}}
 {id=WavePlayer_44, properties {soundWaveAssetPtr=&/Spot_DragonFly_By_A_08}}
 {id=WavePlayer_45, properties {soundWaveAssetPtr=&/Spot_DragonFly_By_A_09}}
 {id=WavePlayer_46, properties {soundWaveAssetPtr=&/Spot_DragonFly_By_A_10}}
 {id=WavePlayer_49, properties {soundWaveAssetPtr=&/Spot_DragonFly_By_A_01}}
 {id=WavePlayer_50, properties {soundWaveAssetPtr=&/Spot_DragonFly_By_A_02}}
 {id=WavePlayer_51, properties {soundWaveAssetPtr=&/Spot_DragonFly_By_A_03}}
 {id=WavePlayer_52, properties {soundWaveAssetPtr=&/Spot_DragonFly_By_A_04}}
 {id=WavePlayer_55, properties {soundWaveAssetPtr=&/Spot_DragonFly_By_A_12}}
 {id=WavePlayer_56, properties {soundWaveAssetPtr=&/Spot_DragonFly_By_A_11}}
 {id=WavePlayer_57, properties {soundWaveAssetPtr=&/Spot_DragonFly_By_A_06}}
 {id=WavePlayer_58, properties {soundWaveAssetPtr=&/Spot_DragonFly_By_A_05}}
]
properties {
 attenuationSettings=&/Spot_Dragonflies_Looping_Oneshots_Att
 bOverrideConcurrency=true
 maxDistance=2500
 priority=0
 soundClassObject=&Fort_Ambient_Athena
 volumeMultiplier=0.251
 busSends [{audioBus=null, minSendLevel=0.6, sendLevel=1, soundSourceBus=&AmbientAudio_Exterior_SourceBus}]
 concurrencyOverrides {maxCount=1, voiceStealReleaseTime=1}
 soundSubmixSends [{minSendLevel=1, soundSubmix=&Fortnite_Reverb_Default}]
}
